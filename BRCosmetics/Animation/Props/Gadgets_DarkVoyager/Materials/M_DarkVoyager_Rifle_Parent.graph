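# Parent material for the Dark Voyager rifle, with emissive glow that reacts to shots, low ammo and reloads.
# Textures supply darkened grayscale base color, surface response, normals and masked, boosted emissive.
# Blueprint-fed timestamps and flags drive a decaying shot flash, a dimmer, less saturated low-ammo pulse and a timed reload fade along pre-skinned local position.
# The final emissive divides by eye adaptation to offset auto exposure.
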Material /BRCosmetics/Animation/Props/Gadgets_DarkVoyager/Materials/M_DarkVoyager_Rifle_Parent
graph {
 properties {
  bAutomaticallySetUsageInEditor=false
  bHasPixelAnimation=true
  bUseMaterialAttributes=true
  bUsedWithClothing=true
  bUsedWithMeshDeformer=true
  bUsedWithMorphTargets=true
  bUsedWithNanite=true
  bUsedWithSkeletalMesh=true
  pixelDepthOffsetMode=PDOM_Legacy
  refractionMethod=RM_IndexOfRefraction
  usedShadingModels=Default Lit
 }
 nodes [
  n1 = Time()
  n2 = ScalarParameter() {parameterName=BP_GetTime_PerShot}
  n3 = Saturate(None=Multiply(A=Subtract(A=@n1, B=@n2)) {constB=3.4})
  n4 = OneMinus(None=@n3)
  n5 = Ceil(None=Frac(None=@n3))
  NamedRerouteDeclaration(None=LinearInterpolate(Alpha=Multiply(A=@n4, B=@n5)) {constA=1, constB=4}) {name=PerShot_Glow, nodeColor=<rgba 0.1719 0 1 1>}
  n6 = Time() {bIgnorePause=true, bOverride_Period=true, period=10}
  n7 = ConstantBiasScale(None=Sine(None=Multiply(A=@n6) {constB=1.3}))
  NamedRerouteDeclaration(None=LinearInterpolate(B=LinearInterpolate(Alpha=@n7) {constA=0.22, constB=0.45}, Alpha=ScalarParameter() {parameterName=BP_HasLowAmmo}) {constA=1, constB=0.5}) {name=LowAmmoGlow, nodeColor=<rgba 1 0 0.8437 1>}
  n8 = LocalPosition() {localOrigin=InstancePreSkinning}
  NamedRerouteDeclaration(None=ShaderStageSwitch(PixelShader=VertexInterpolator(VS=@n8.XYZ).PS, VertexShader=@n8.XYZ)) {name=PS_LocalPosition, nodeColor=<rgba 0 0.5312 1 1>}
  n9 = TextureSample() {samplerSource=SSM_Wrap_WorldGroupSettings, texture=/BRCosmetics/Animation/Props/Gadgets_DarkVoyager/Textures/T_Rifle_DarkVoyager_D}
  n10 = Constant() {r=1}
  n11 = TextureSample() {
   samplerSource=SSM_Wrap_WorldGroupSettings
   samplerType=SAMPLERTYPE_Masks
   texture=/BRCosmetics/Animation/Props/Gadgets_DarkVoyager/Textures/T_Rifle_DarkVoyager_S
  }
  n12 = OneMinus(None=@n11.R)
  n13 = Multiply(
   A=Multiply(A=@n12, B=@n12)
   B=TextureSample() {samplerSource=SSM_Wrap_WorldGroupSettings, texture=/BRCosmetics/Animation/Props/Gadgets_DarkVoyager/Textures/T_Rifle_DarkVoyager_E}.RGB
  )
  n14 = ScalarParameter() {parameterName=BP_HasLowAmmo}
  n15 = TextureSampleParameter2D() {
   parameterName=NormalMap
   samplerSource=SSM_Wrap_WorldGroupSettings
   samplerType=SAMPLERTYPE_Normal
   texture=/BRCosmetics/Animation/Props/Gadgets_DarkVoyager/Textures/T_Rifle_DarkVoyager_LOD1_N
  }
  n16 = SetMaterialAttributes!(
   MaterialAttributes
   Base Color=Multiply(A=Desaturation(None=@n9.RGB, Fraction=@n10)) {constB=0.1}
   Metallic=@n11.G
   Specular=@n11.G
   Roughness=@n11.G
   Emissive Color=Desaturation(None=Multiply(A=@n13) {constB=25}, Fraction=LinearInterpolate(Alpha=@n14) {constA=-1, constB=0})
   Normal=@n15.RGB
  ) {attributeSetTypes [
    #1
    #2
    #3
    #4
    #5
    #6
   ]}
  n17 = GetMaterialAttributes(None=@n16) {attributeGetTypes [#5, #1]}
  n18 = ComponentMask(None=NamedRerouteUsage()) {g=true}
  n19 = SphereMask(A=@n18, B=Constant() {r=35}) {attenuationRadius=10}
  n20 = Time()
  n21 = ScalarParameter() {defaultValue=9999999827968, parameterName=BP_GetTime}
  n22 = ScalarParameter() {defaultValue=1, parameterName=BP_ReloadDuration}
  n23 = Saturate(None=Divide(A=Subtract(A=@n20, B=@n21), B=Max(A=@n22) {constB=0.01}))
  n24 = Multiply(A=SmoothStep(Value=@n23) {constMax=0.1}, B=OneMinus(None=SmoothStep(Value=@n23) {constMax=0.9, constMin=0.8}))
  n25 = Add(A=@n18, B=LinearInterpolate(Alpha=@n24) {constA=-20, constAlpha=0, constB=-40})
  n26 = Add(
   A=Multiply(A=OneMinus(None=@n19), B=OneMinus(None=@n24))
   B=NamedRerouteDeclaration(None=Saturate(None=@n25)) {name=FadeMask, nodeColor=<rgba 0 1 0.5078 1>}
  )
  n27 = Multiply(A=Multiply(A=@n17.Emissive Color, B=Saturate(None=@n26)), B=NamedRerouteUsage())
  n28 = NamedRerouteUsage()
  SetMaterialAttributes!(MaterialAttributes=@n17.MaterialAttributes, Emissive Color=Divide(A=Multiply(A=@n27, B=@n28), B=EyeAdaptation().EyeAdaptation), Base Color) {attributeSetTypes [#5, #1]}
 ]
}
